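# Physics asset for the Weeping Woods Fall male one-handed axe pickaxe, setting up physical motion for four leaf bones on the weapon.
# Small capsule bodies cover the handle bone and each dyn_leaf bone.
# Each leaf is constrained directly to the handle in a star layout rather than a chain, with matching swing and twist limit angles so each leaf hangs from the handle independently.
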PhysicsAsset /BRCosmetics/Weapons/FORT_Melee/Pickaxe_WeepingWoods_Fall_Male_1h/Meshes/Weeping_Woods_Fall_Male_1h_Axe_Physics
bodies {
 Melee_TwoHanded_Handle [{shapeType=Capsule, length=1, radius=1}]
 dyn_leaf_1 [{shapeType=Capsule, length=2, radius=1, center=<1 0 0>, rotation=<rot 90 0 -45>}]
 dyn_leaf_2 [{shapeType=Capsule, length=1.5, radius=0.75, center=<0.7129 0.1984 0>, rotation=<rot 0 -34.644 -75.6941>}]
 dyn_leaf_3 [{shapeType=Capsule, length=2, radius=1, center=<0.7391 0.187 0>, rotation=<rot 0 -35.6308 -83.963>}]
 dyn_leaf_4 [{shapeType=Capsule, length=2.6527, radius=1.1679, center=<0.3239 0.683 1.0762>, rotation=<rot 29.04 -89.9999 10.5403>}]
}
constraints [
 {bone1Name=dyn_leaf_1, bone2Name=Melee_TwoHanded_Handle, swing1LimitDegrees=45, swing2LimitDegrees=45, twistLimitDegrees=45}
 {bone1Name=dyn_leaf_2, bone2Name=Melee_TwoHanded_Handle, swing1LimitDegrees=45, swing2LimitDegrees=45, twistLimitDegrees=45}
 {bone1Name=dyn_leaf_3, bone2Name=Melee_TwoHanded_Handle, swing1LimitDegrees=45, swing2LimitDegrees=45, twistLimitDegrees=45}
 {bone1Name=dyn_leaf_4, bone2Name=Melee_TwoHanded_Handle, swing1LimitDegrees=45, swing2LimitDegrees=45, twistLimitDegrees=45}
]
